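# Sound cue for picking up an object in the Season 21 Week 6 narrative quest content.
# A random node picks one of three pickup wave variations with equal weight and feeds a modulator that adds slight pitch and volume variation.
# It plays in the quests sound class with the Week 6 attenuation asset for spatial falloff.
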SoundCue /BattlePassS21/QuestActors/Narrative/Week_06/SFX/S21_W06_ObjectPickup_Cue
root=Modulator_0
nodes [
 {id=Modulator_0, children [Random_0], properties {pitchMax=1.01, pitchMin=0.98, volumeMax=1, volumeMin=0.99}}
 {id=Random_0, children [WavePlayer_0,WavePlayer_1,WavePlayer_2], properties {weights [1,1,1]}}
 {id=WavePlayer_0, properties {soundWaveAssetPtr=&/S21_W06_ObjectPickup_01}}
 {id=WavePlayer_1, properties {soundWaveAssetPtr=&/S21_W06_ObjectPickup_02}}
 {id=WavePlayer_2, properties {soundWaveAssetPtr=&/S21_W06_ObjectPickup_03}}
]
properties {
 attenuationSettings=&/S21_W06_Atten
 maxDistance=1300
 soundClassObject=&Fort_Quests
 volumeMultiplier=1
}
