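# Looping sound cue for a quest object in Battle Pass Season 21 narrative Week 06, routed through the Fort_Quests sound class.
# A single looping wave player feeds the S21_W06_Object_LP wave straight to the root.
# The cue lowers pitch and uses the S21_W06_Atten attenuation settings to limit how far it carries.
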SoundCue /BattlePassS21/QuestActors/Narrative/Week_06/SFX/S21_W06_Object_LP_Cue
root=WavePlayer_0
nodes [{id=WavePlayer_0, properties {bLooping=true, soundWaveAssetPtr=&/S21_W06_Object_LP}}]
properties {
 attenuationSettings=&/S21_W06_Atten
 maxDistance=1300
 pitchMultiplier=0.8
 soundClassObject=&Fort_Quests
}
