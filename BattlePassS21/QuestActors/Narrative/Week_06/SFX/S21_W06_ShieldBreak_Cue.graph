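# Sound cue for a shield-break effect in the Season 21 Battle Pass week 6 narrative quest content.
# A Random node picks one of five ShieldBreak wave variations with equal weight, then a Modulator at default settings passes it to the output.
# It uses the week 6 attenuation asset and plays under the Fort_Quests sound class.
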SoundCue /BattlePassS21/QuestActors/Narrative/Week_06/SFX/S21_W06_ShieldBreak_Cue
root=Modulator_0
nodes [
 {id=Modulator_0, children [Random_0]}
 {id=Random_0, children [WavePlayer_0,WavePlayer_1,WavePlayer_2,WavePlayer_3,WavePlayer_4], properties {weights [1,1,1,1,1]}}
 {id=WavePlayer_0, properties {soundWaveAssetPtr=&/S21_W06_ShieldBreak_01}}
 {id=WavePlayer_1, properties {soundWaveAssetPtr=&/S21_W06_ShieldBreak_02}}
 {id=WavePlayer_2, properties {soundWaveAssetPtr=&/S21_W06_ShieldBreak_03}}
 {id=WavePlayer_3, properties {soundWaveAssetPtr=&/S21_W06_ShieldBreak_04}}
 {id=WavePlayer_4, properties {soundWaveAssetPtr=&/S21_W06_ShieldBreak_05}}
]
properties {
 attenuationSettings=&/S21_W06_Atten
 maxDistance=1300
 soundClassObject=&Fort_Quests
}
